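# Control Rig for the SquidGlisten LiftDairy medium male body that procedurally poses its shoulder armour, straps and belt bones.
# Forward solve runs the shared DeformRig function, then Shoulders_Setup and Belts_Setup, both built on weighted multi-bone parent constraints.
# Shoulders add corrective offsets driven by DeformRig upper-arm alpha attributes.
# Belts rotate from curve-remapped thigh swing (swing-twist against reference controls), adjusted by emote and crouch-movement variables.
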ControlRigBlueprint /BRCosmetics/Characters/Player/Male/Medium/Bodies/M_MED_SquidGlisten_LiftDairy/Meshes/M_MED_SquidGlisten_LiftDairy_CtrlRig
graphs [
 New Function_ContainedGraph {nodes [
   RigVMFunctionEntryNode()
   > RigVMFunction_Sequence = RigVMAggregateNode()
   &RigUnit_ParentConstraint(
    AdvancedSettings=(RotationOrderForFilter=XZY)
    Child=(Type=Bone,Name="base_armourR")
    Parents="((Item=(Type=Bone,Name=\"upperarm_r\"),Weight=0.0109),(Item=(Type=Bone,Name=\"spine_03\"),Weight=0.016),(Item=(Type=Bone,Name=\"spine_05\"),Weight=0.0292),(Item=(Type=Bone,Name=\"deform_upperarm_fr_r\"),Weight=0.0752),(Item=(Type=Bone,Name=\"upperarm_twist_02_r\"),W ...(truncated)"
    Weight=1
    ExecutePin=@RigVMFunction_Sequence.A
   )
   > &RigUnit_ParentConstraint(
    AdvancedSettings=(RotationOrderForFilter=XZY)
    Child=(Type=Bone,Name="base_shoulder_strap_R")
    Parents="((Item=(Type=Bone,Name=\"deform_pec_r\"),Weight=0.0023),(Item=(Type=Bone,Name=\"deform_upperarm_fr_r\"),Weight=0.0162),(Item=(Type=Bone,Name=\"spine_04\"),Weight=0.018),(Item=(Type=Bone,Name=\"spine_03\"),Weight=0.0254),(Item=(Type=Bone,Name=\"upperarm_twist_01_r\") ...(truncated)"
    Weight=1
   )
   > &RigUnit_ParentConstraint(
    AdvancedSettings=(RotationOrderForFilter=XZY)
    Child=(Type=Bone,Name="base_arm_strap_R")
    Parents="((Item=(Type=Bone,Name=\"deform_elbow_in_r\"),Weight=0.0003),(Item=(Type=Bone,Name=\"deform_upperarm_up_r\"),Weight=0.0003),(Item=(Type=Bone,Name=\"lowerarm_r\"),Weight=0.0009),(Item=(Type=Bone,Name=\"clavicle_r\"),Weight=0.0038),(Item=(Type=Bone,Name=\"deform_uppe ...(truncated)"
    Weight=1
   )
   > &RigUnit_ParentConstraint(
    AdvancedSettings=(RotationOrderForFilter=XZY)
    Child=(Type=Bone,Name="base_shoulder_strap_L")
    Parents="((Item=(Type=Bone,Name=\"deform_pec_l\"),Weight=0.0023),(Item=(Type=Bone,Name=\"deform_upperarm_fr_l\"),Weight=0.0161),(Item=(Type=Bone,Name=\"spine_04\"),Weight=0.0179),(Item=(Type=Bone,Name=\"spine_03\"),Weight=0.0253),(Item=(Type=Bone,Name=\"upperarm_twist_01_l\" ...(truncated)"
    Weight=1
   )
   > &RigUnit_ParentConstraint(
    AdvancedSettings=(RotationOrderForFilter=XZY)
    Child=(Type=Bone,Name="base_armourL")
    Parents="((Item=(Type=Bone,Name=\"upperarm_l\"),Weight=0.0124),(Item=(Type=Bone,Name=\"spine_03\"),Weight=0.0164),(Item=(Type=Bone,Name=\"spine_05\"),Weight=0.0307),(Item=(Type=Bone,Name=\"deform_upperarm_fr_l\"),Weight=0.0574),(Item=(Type=Bone,Name=\"upperarm_twist_02_l\"), ...(truncated)"
    Weight=1
   )
   > &RigUnit_ParentConstraint(
    AdvancedSettings=(RotationOrderForFilter=XZY)
    Child=(Type=Bone,Name="base_arm_strap_L")
    Parents="((Item=(Type=Bone,Name=\"deform_elbow_in_l\"),Weight=0.0003),(Item=(Type=Bone,Name=\"deform_upperarm_up_l\"),Weight=0.0003),(Item=(Type=Bone,Name=\"lowerarm_l\"),Weight=0.0009),(Item=(Type=Bone,Name=\"clavicle_l\"),Weight=0.0038),(Item=(Type=Bone,Name=\"deform_uppe ...(truncated)"
    Weight=1
   )
   RigVMFunctionReturnNode(ExecuteContext=@RigVMFunction_Sequence.C)
   n1 = &RigVMFunction_MathDoubleRemap(SourceMaximum=1, SourceMinimum=0, TargetMaximum=0, TargetMinimum=1, Value=0, bClamp=false)
   &RigUnit_ParentConstraint(
    AdvancedSettings=(InterpolationType=Average,RotationOrderForFilter=XZY)
    Child=(Type=Bone,Name="base_armourL")
    Filter=(TranslationFilter=(bX=false,bY=false,bZ=false),RotationFilter=(bX=True,bY=True,bZ=True),ScaleFilter=(bX=false,bY=false,bZ=false))
    Parents="((Item=(Type=Bone,Name=\"deform_upperarm_up_l\"),Weight=0.25),(Item=(Type=Bone,Name=\"upperarm_twist_01_l\"),Weight=0.5),(Item=(Type=Bone,Name=\"upperarm_l\"),Weight=0.25),(Item=(Type=Bone,Name=\"deform_upperarm_fr_l\"),Weight=0.4),(Item=(Type=Bone,Name=\"deform_up ...(truncated)"
    bMaintainOffset=true
    Weight=@n1.Result
    ExecutePin=@RigVMFunction_Sequence.B
   )
   n2 = RigVMDispatchNode(BoneName=root, Default=0, Name=deformRig_l_upperarm_fr_alpha) {template="DISPATCH_RigDispatch_GetAnimAttribute(in Name,in BoneName,in Default,out Value,out Found)"}
   Get_Animation_Attribute = RigVMDispatchNode(BoneName=root, Default=0, Name=deformRig_l_upperarm_bk_alpha) {template="DISPATCH_RigDispatch_GetAnimAttribute(in Name,in BoneName,in Default,out Value,out Found)"}
   Add = &RigVMFunction_MathDoubleAdd(A=@n2.Value, B=@Get_Animation_Attribute.Value)
   Multiply = &RigVMFunction_MathDoubleMul(B=25, A=@n2.Value)
   RigVMFunction_MathQuaternionFromEuler = &RigVMFunction_MathQuaternionFromEuler(
    Euler=(X=0,Y=5,Z=15)
    Euler.X=&RigVMFunction_MathDoubleAdd(A=@Multiply.Result, B=&RigVMFunction_MathDoubleMul(B=25, A=@Get_Animation_Attribute.Value).Result).Result
   )
   > &RigUnit_ModifyTransforms(
    ItemToModify=((Item=(Type=Bone,Name="dyn_armourL"),Transform=(Rotation=(X=0,Y=0,Z=0,W=1),Translation=(X=2,Y=0.25,Z=0.6),Scale3D=(X=0.9,Y=1,Z=1))))
    WeightMinimum=0
    Weight=&RigVMFunction_MathDoubleMul(A=@n1.Result, B=@Add.Result).Result
    ItemToModify.0.Transform.Rotation=@RigVMFunction_MathQuaternionFromEuler.Result
   )
   Subtract = &RigVMFunction_MathDoubleSub(
    B=@n2.Value
    A=RigVMDispatchNode(BoneName=root, Default=0, Name=deformRig_l_upperarm_up_alpha) {template="DISPATCH_RigDispatch_GetAnimAttribute(in Name,in BoneName,in Default,out Value,out Found)"}.Value
   )
   > &RigUnit_ModifyTransforms(
    ItemToModify=((Item=(Type=Bone,Name="dyn_armourL"),Transform=(Rotation=(X=-0.0019,Y=0.0436,Z=0.0436,W=0.9981),Translation=(X=1,Y=-0.4,Z=1.5),Scale3D=(X=0.9,Y=1.05,Z=1))))
    WeightMinimum=0
    Weight=&RigVMFunction_MathDoubleMul(A=@n1.Result, B=@Subtract.Result).Result
   )
   > &RigUnit_ModifyTransforms(
    ItemToModify=((Item=(Type=Bone,Name="base_shoulder_strap_L"),Transform=(Rotation=(X=0,Y=0,Z=0,W=1),Translation=(X=-1,Y=0,Z=1.5),Scale3D=(X=0.95,Y=1,Z=1))))
    WeightMinimum=0
    Weight=RigVMDispatchNode(BoneName=root, Default=0, Name=deformRig_l_upperarm_up_alpha) {template="DISPATCH_RigDispatch_GetAnimAttribute(in Name,in BoneName,in Default,out Value,out Found)"}.Value
   )
   n3 = &RigVMFunction_MathDoubleRemap(SourceMaximum=1, SourceMinimum=0, TargetMaximum=0, TargetMinimum=1, Value=0, bClamp=False)
   > &RigUnit_ParentConstraint(
    AdvancedSettings=(InterpolationType=Average,RotationOrderForFilter=XZY)
    Child=(Type=Bone,Name="base_armourR")
    Filter=(TranslationFilter=(bX=false,bY=false,bZ=false),RotationFilter=(bX=True,bY=True,bZ=True),ScaleFilter=(bX=false,bY=false,bZ=false))
    Parents="((Item=(Type=Bone,Name=\"deform_upperarm_up_r\"),Weight=0.25),(Item=(Type=Bone,Name=\"upperarm_twist_01_r\"),Weight=0.5),(Item=(Type=Bone,Name=\"upperarm_r\"),Weight=0.25),(Item=(Type=Bone,Name=\"deform_upperarm_fr_r\"),Weight=0.4),(Item=(Type=Bone,Name=\"deform_up ...(truncated)"
    bMaintainOffset=true
    Weight=@n3.Result
   )
   n4 = RigVMDispatchNode(BoneName=root, Default=0, Name=deformRig_r_upperarm_fr_alpha) {template="DISPATCH_RigDispatch_GetAnimAttribute(in Name,in BoneName,in Default,out Value,out Found)"}
   Get_Animation_Attribute_2 = RigVMDispatchNode(BoneName=root, Default=0, Name=deformRig_r_upperarm_bk_alpha) {template="DISPATCH_RigDispatch_GetAnimAttribute(in Name,in BoneName,in Default,out Value,out Found)"}
   n5 = &RigVMFunction_MathDoubleMul(B=-35, A=@n4.Value)
   n6 = &RigVMFunction_MathDoubleMul(B=-35, A=@Get_Animation_Attribute_2.Value)
   > &RigUnit_ModifyTransforms(
    ItemToModify=((Item=(Type=Bone,Name="dyn_armourR"),Transform=(Rotation=(X=0,Y=0,Z=0,W=1),Translation=(X=2.75,Y=0.1,Z=0.6),Scale3D=(X=0.85,Y=1,Z=1))))
    WeightMinimum=0
    Weight=&RigVMFunction_MathDoubleMul(A=@n3.Result, B=&RigVMFunction_MathDoubleAdd(A=@n4.Value, B=@Get_Animation_Attribute_2.Value).Result).Result
    ItemToModify.0.Transform.Rotation=&RigVMFunction_MathQuaternionFromEuler(Euler=(X=0,Y=17,Z=-20), Euler.X=&RigVMFunction_MathDoubleAdd(A=@n5.Result, B=@n6.Result).Result).Result
   )
   n7 = RigVMDispatchNode(BoneName=root, Default=0, Name=deformRig_r_upperarm_up_alpha) {template="DISPATCH_RigDispatch_GetAnimAttribute(in Name,in BoneName,in Default,out Value,out Found)"}
   > &RigUnit_ModifyTransforms(
    ItemToModify=((Item=(Type=Bone,Name="dyn_armourR"),Transform=(Rotation=(X=-0.0019,Y=0.0436,Z=0.0436,W=0.9981),Translation=(X=1.5,Y=0,Z=1),Scale3D=(X=0.9,Y=1.1,Z=1))))
    WeightMinimum=0
    Weight=&RigVMFunction_MathDoubleMul(A=@n3.Result, B=&RigVMFunction_MathDoubleSub(A=@n7.Value, B=@n4.Value).Result).Result
   )
   > &RigUnit_ModifyTransforms(
    ItemToModify=((Item=(Type=Bone,Name="base_shoulder_strap_R"),Transform=(Rotation=(X=0,Y=0,Z=0,W=1),Translation=(X=-1,Y=0,Z=1.5),Scale3D=(X=0.95,Y=1,Z=1))))
    WeightMinimum=0
    Weight=RigVMDispatchNode(BoneName=root, Default=0, Name=deformRig_r_upperarm_up_alpha) {template="DISPATCH_RigDispatch_GetAnimAttribute(in Name,in BoneName,in Default,out Value,out Found)"}.Value
   )
   RigVMFunction_ControlFlowBranch = &RigVMFunction_ControlFlowBranch(Condition=False, ExecuteContext=())
   &RigVMFunction_ControlFlowBranch(Condition=False, ExecuteContext=())
  ]}
 New Function_ContainedGraph {nodes [
   RigVMFunctionEntryNode()
   > RigVMFunction_Sequence = RigVMAggregateNode()
   &RigUnit_ParentConstraint(
    AdvancedSettings=(RotationOrderForFilter=XZY)
    Child=(Type=Bone,Name="base_beltRold_l_06")
    Parents=((Item=(Type=Bone,Name="spine_04"),Weight=0.0127),(Item=(Type=Bone,Name="spine_03"),Weight=0.114),(Item=(Type=Bone,Name="spine_01"),Weight=0.1592),(Item=(Type=Bone,Name="spine_02"),Weight=0.1894),(Item=(Type=Bone,Name="pelvis"),Weight=0.5248))
    Weight=1
    ExecutePin=@RigVMFunction_Sequence.A
   )
   > &RigUnit_ParentConstraint(
    AdvancedSettings=(RotationOrderForFilter=XZY)
    Child=(Type=Bone,Name="base_beltRold_l_05")
    Parents="((Item=(Type=Bone,Name=\"thigh_twist_01_l\"),Weight=0.0005),(Item=(Type=Bone,Name=\"deform_groin_l\"),Weight=0.0007),(Item=(Type=Bone,Name=\"spine_04\"),Weight=0.0082),(Item=(Type=Bone,Name=\"spine_03\"),Weight=0.1053),(Item=(Type=Bone,Name=\"spine_01\"),Weight=0.14 ...(truncated)"
    Weight=1
   )
   > &RigUnit_ParentConstraint(
    AdvancedSettings=(RotationOrderForFilter=XZY)
    Child=(Type=Bone,Name="base_beltRold_l_04")
    Parents=((Item=(Type=Bone,Name="spine_04"),Weight=0.0025),(Item=(Type=Bone,Name="spine_03"),Weight=0.1061),(Item=(Type=Bone,Name="spine_01"),Weight=0.1511),(Item=(Type=Bone,Name="spine_02"),Weight=0.1808),(Item=(Type=Bone,Name="pelvis"),Weight=0.5595))
    Weight=1
   )
   > &RigUnit_ParentConstraint(
    AdvancedSettings=(RotationOrderForFilter=XZY)
    Child=(Type=Bone,Name="base_beltRold_l_03")
    Parents=((Item=(Type=Bone,Name="spine_03"),Weight=0.0987),(Item=(Type=Bone,Name="spine_01"),Weight=0.1459),(Item=(Type=Bone,Name="spine_02"),Weight=0.1749),(Item=(Type=Bone,Name="pelvis"),Weight=0.5805))
    Weight=1
   )
   > &RigUnit_ParentConstraint(
    AdvancedSettings=(RotationOrderForFilter=XZY)
    Child=(Type=Bone,Name="base_beltRold_l_02")
    Parents="((Item=(Type=Bone,Name=\"spine_04\"),Weight=0.0004),(Item=(Type=Bone,Name=\"deform_groin_l\"),Weight=0.002),(Item=(Type=Bone,Name=\"deform_glute_l\"),Weight=0.005),(Item=(Type=Bone,Name=\"thigh_twist_01_l\"),Weight=0.0132),(Item=(Type=Bone,Name=\"spine_03\"),Weight= ...(truncated)"
    Weight=1
   )
   > &RigUnit_ParentConstraint(
    AdvancedSettings=(RotationOrderForFilter=XZY)
    Child=(Type=Bone,Name="base_beltRold_l_01")
    Parents="((Item=(Type=Bone,Name=\"spine_04\"),Weight=0.0003),(Item=(Type=Bone,Name=\"thigh_twist_01_l\"),Weight=0.0013),(Item=(Type=Bone,Name=\"deform_glute_l\"),Weight=0.0056),(Item=(Type=Bone,Name=\"spine_03\"),Weight=0.0384),(Item=(Type=Bone,Name=\"spine_01\"),Weight=0.05 ...(truncated)"
    Weight=1
   )
   > &RigUnit_ParentConstraint(
    AdvancedSettings=(RotationOrderForFilter=XZY)
    Child=(Type=Bone,Name="dyn_beltRold_l_04_b")
    Filter=(TranslationFilter=(bX=false,bY=false,bZ=false),RotationFilter=(bX=True,bY=True,bZ=True),ScaleFilter=(bX=false,bY=false,bZ=false))
    Parents="((Item=(Type=Bone,Name=\"thigh_twist_01_l\"),Weight=0.019),(Item=(Type=Bone,Name=\"deform_groin_l\"),Weight=0.0191),(Item=(Type=Bone,Name=\"spine_03\"),Weight=0.0445),(Item=(Type=Bone,Name=\"spine_01\"),Weight=0.0677),(Item=(Type=Bone,Name=\"spine_02\"),Weight=0.078 ...(truncated)"
    Weight=1
   )
   > &RigUnit_ParentConstraint(
    AdvancedSettings=(RotationOrderForFilter=XZY)
    Child=(Type=Bone,Name="dyn_beltRold_l_03_b")
    Filter=(TranslationFilter=(bX=false,bY=false,bZ=false),RotationFilter=(bX=True,bY=True,bZ=True),ScaleFilter=(bX=false,bY=false,bZ=false))
    Parents="((Item=(Type=Bone,Name=\"deform_glute_l\"),Weight=0.0039),(Item=(Type=Bone,Name=\"spine_03\"),Weight=0.0209),(Item=(Type=Bone,Name=\"deform_groin_l\"),Weight=0.0261),(Item=(Type=Bone,Name=\"spine_01\"),Weight=0.0333),(Item=(Type=Bone,Name=\"spine_02\"),Weight=0.0388 ...(truncated)"
    Weight=1
   )
   > &RigUnit_ParentConstraint(
    AdvancedSettings=(RotationOrderForFilter=XZY)
    Child=(Type=Bone,Name="dyn_beltRold_l_05_b")
    Filter=(TranslationFilter=(bX=false,bY=false,bZ=false),RotationFilter=(bX=True,bY=True,bZ=True),ScaleFilter=(bX=false,bY=false,bZ=false))
    Parents="((Item=(Type=Bone,Name=\"spine_04\"),Weight=0.001),(Item=(Type=Bone,Name=\"thigh_twist_01_l\"),Weight=0.0119),(Item=(Type=Bone,Name=\"deform_groin_l\"),Weight=0.0159),(Item=(Type=Bone,Name=\"spine_03\"),Weight=0.0555),(Item=(Type=Bone,Name=\"spine_01\"),Weight=0.080 ...(truncated)"
    Weight=1
   )
   > &RigUnit_ParentConstraint(
    AdvancedSettings=(RotationOrderForFilter=XZY)
    Child=(Type=Bone,Name="dyn_beltRold_l_06_b")
    Filter=(TranslationFilter=(bX=false,bY=false,bZ=false),RotationFilter=(bX=True,bY=True,bZ=True),ScaleFilter=(bX=false,bY=false,bZ=false))
    Parents="((Item=(Type=Bone,Name=\"deform_groin_r\"),Weight=0.0008),(Item=(Type=Bone,Name=\"thigh_twist_01_r\"),Weight=0.0015),(Item=(Type=Bone,Name=\"deform_groin_l\"),Weight=0.0029),(Item=(Type=Bone,Name=\"thigh_twist_01_l\"),Weight=0.0031),(Item=(Type=Bone,Name=\"spine_03 ...(truncated)"
    Weight=1
   )
   > &RigUnit_ParentConstraint(
    AdvancedSettings=(RotationOrderForFilter=XZY)
    Child=(Type=Bone,Name="base_beltRold_r_05")
    Parents="((Item=(Type=Bone,Name=\"deform_groin_l\"),Weight=0.0002),(Item=(Type=Bone,Name=\"thigh_twist_01_l\"),Weight=0.0004),(Item=(Type=Bone,Name=\"thigh_twist_01_r\"),Weight=0.0076),(Item=(Type=Bone,Name=\"deform_groin_r\"),Weight=0.0096),(Item=(Type=Bone,Name=\"spine_03 ...(truncated)"
    Weight=1
   )
   > &RigUnit_ParentConstraint(
    AdvancedSettings=(RotationOrderForFilter=XZY)
    Child=(Type=Bone,Name="base_beltRold_r_04")
    Parents=((Item=(Type=Bone,Name="spine_04"),Weight=0.0025),(Item=(Type=Bone,Name="spine_03"),Weight=0.1061),(Item=(Type=Bone,Name="spine_01"),Weight=0.1511),(Item=(Type=Bone,Name="spine_02"),Weight=0.1808),(Item=(Type=Bone,Name="pelvis"),Weight=0.5595))
    Weight=1
   )
   > &RigUnit_ParentConstraint(
    AdvancedSettings=(RotationOrderForFilter=XZY)
    Child=(Type=Bone,Name="base_beltRold_r_03")
    Parents=((Item=(Type=Bone,Name="spine_03"),Weight=0.0987),(Item=(Type=Bone,Name="spine_01"),Weight=0.1459),(Item=(Type=Bone,Name="spine_02"),Weight=0.1749),(Item=(Type=Bone,Name="pelvis"),Weight=0.5805))
    Weight=1
   )
   > &RigUnit_ParentConstraint(
    AdvancedSettings=(RotationOrderForFilter=XZY)
    Child=(Type=Bone,Name="base_beltRold_r_02")
    Parents="((Item=(Type=Bone,Name=\"spine_04\"),Weight=0.0004),(Item=(Type=Bone,Name=\"deform_groin_r\"),Weight=0.002),(Item=(Type=Bone,Name=\"deform_glute_r\"),Weight=0.005),(Item=(Type=Bone,Name=\"thigh_twist_01_r\"),Weight=0.0132),(Item=(Type=Bone,Name=\"spine_03\"),Weight= ...(truncated)"
    Weight=1
   )
   > &RigUnit_ParentConstraint(
    AdvancedSettings=(RotationOrderForFilter=XZY)
    Child=(Type=Bone,Name="base_beltRold_r_01")
    Parents="((Item=(Type=Bone,Name=\"deform_groin_r\"),Weight=0.0003),(Item=(Type=Bone,Name=\"spine_04\"),Weight=0.0008),(Item=(Type=Bone,Name=\"thigh_twist_01_r\"),Weight=0.0046),(Item=(Type=Bone,Name=\"deform_glute_r\"),Weight=0.0065),(Item=(Type=Bone,Name=\"spine_03\"),Weigh ...(truncated)"
    Weight=1
   )
   > &RigUnit_ParentConstraint(
    AdvancedSettings=(RotationOrderForFilter=XZY)
    Child=(Type=Bone,Name="dyn_beltRold_r_04_b")
    Filter=(TranslationFilter=(bX=false,bY=false,bZ=false),RotationFilter=(bX=True,bY=True,bZ=True),ScaleFilter=(bX=false,bY=false,bZ=false))
    Parents="((Item=(Type=Bone,Name=\"thigh_twist_01_r\"),Weight=0.019),(Item=(Type=Bone,Name=\"deform_groin_r\"),Weight=0.0191),(Item=(Type=Bone,Name=\"spine_03\"),Weight=0.0445),(Item=(Type=Bone,Name=\"spine_01\"),Weight=0.0677),(Item=(Type=Bone,Name=\"spine_02\"),Weight=0.078 ...(truncated)"
    Weight=1
   )
   > &RigUnit_ParentConstraint(
    AdvancedSettings=(RotationOrderForFilter=XZY)
    Child=(Type=Bone,Name="dyn_beltRold_r_03_b")
    Filter=(TranslationFilter=(bX=false,bY=false,bZ=false),RotationFilter=(bX=True,bY=True,bZ=True),ScaleFilter=(bX=false,bY=false,bZ=false))
    Parents="((Item=(Type=Bone,Name=\"deform_glute_r\"),Weight=0.0008),(Item=(Type=Bone,Name=\"deform_groin_r\"),Weight=0.01),(Item=(Type=Bone,Name=\"thigh_twist_01_r\"),Weight=0.0226),(Item=(Type=Bone,Name=\"spine_03\"),Weight=0.0415),(Item=(Type=Bone,Name=\"spine_01\"),Weight= ...(truncated)"
    Weight=1
   )
   n1 = &RigUnit_GetRelativeTransformForItem(Child=(Type=Bone,Name="thigh_twist_01_r"), Parent=(Type=Control,Name="thigh_twist_01_r_ctrl"), bChildInitial=False, bParentInitial=False)
   RigVMFunction_MathQuaternionSwingTwist = &RigVMFunction_MathQuaternionSwingTwist(TwistAxis=(X=1,Y=0,Z=0), Input=@n1.RelativeTransform.Rotation)
   RigVMFunction_MathQuaternionToEuler = &RigVMFunction_MathQuaternionToEuler(RotationOrder=ZYX, Value=@RigVMFunction_MathQuaternionSwingTwist.Swing)
   RigVMVariableNode(Value=@RigVMFunction_MathQuaternionToEuler.Result, ExecuteContext=@RigVMFunction_Sequence.B) {variable=Leg_r_Swing_rot}
   n2 = &RigUnit_GetRelativeTransformForItem(Child=(Type=Bone,Name="thigh_twist_01_l"), Parent=(Type=Control,Name="thigh_twist_01_l_ctrl"), bChildInitial=False, bParentInitial=False)
   > RigVMVariableNode(Value=&RigVMFunction_MathQuaternionToEuler(
     RotationOrder=ZYX
     Value=&RigVMFunction_MathQuaternionSwingTwist(TwistAxis=(X=1,Y=0,Z=0), Input=@n2.RelativeTransform.Rotation).Swing
    ).Result) {variable=Leg_l_Swing_rot}
   RigVMFunctionReturnNode(ExecuteContext=@RigVMFunction_Sequence.D)
   n3 = RigVMVariableNode() {variable=Leg_l_Swing_rot}
   n4 = &RigVMFunction_AnimEvalRichCurve(
    Curve="(EditorCurveData=(Keys=((InterpMode=RCIM_Cubic,TangentMode=RCTM_User,TangentWeightMode=RCTWM_WeightedNone,Time=0,Value=0,ArriveTangent=0.2796,ArriveTangentWeight=0,LeaveTangent=0.2796,LeaveTangentWeight=0),(InterpMode=RCIM_Cubic,TangentMode=RCTM_User,Tange ...(truncated)"
    SourceMaximum=-140
    SourceMinimum=0
    TargetMaximum=-44
    TargetMinimum=0
    Value=@n3.Value.Y
   )
   n5 = &RigVMFunction_AnimEvalRichCurve(
    Curve="(EditorCurveData=(Keys=((InterpMode=RCIM_Cubic,TangentMode=RCTM_User,TangentWeightMode=RCTWM_WeightedNone,Time=0,Value=0,ArriveTangent=0.2796,ArriveTangentWeight=0,LeaveTangent=0.2796,LeaveTangentWeight=0),(InterpMode=RCIM_Cubic,TangentMode=RCTM_User,Tange ...(truncated)"
    SourceMaximum=-140
    SourceMinimum=0
    TargetMaximum=-44
    TargetMinimum=0
    Value=@n3.Value.Z
   )
   n6 = &RigVMFunction_AlphaInterp(
    InterpSpeedDecreasing=10
    InterpSpeedIncreasing=10
    Scale=1
    bClampResult=false
    bInterpResult=false
    bMapRange=true
    Value=&RigVMFunction_MathDoubleAdd(A=@n4.Result, B=@n5.Result).Result
   )
   n7 = &RigVMFunction_MathQuaternionFromEuler(Euler=(X=0,Y=0,Z=-8), Euler.Y=&RigVMFunction_MathDoubleMul(B=0.95, A=@n6.Result).Result)
   n8 = &RigVMFunction_MathQuaternionFromEuler(Euler=(X=0,Y=0,Z=0), Euler.Y=&RigVMFunction_MathDoubleMul(B=1.75, A=@n6.Result).Result)
   &RigUnit_ModifyTransforms(
    ItemToModify="((Item=(Type=Bone,Name=\"dyn_beltRold_l_05_a\"),Transform=(Rotation=(X=0,Y=-0.1736,Z=0,W=0.9848),Translation=(X=0,Y=0,Z=0),Scale3D=(X=1,Y=1,Z=1))),(Item=(Type=Bone,Name=\"dyn_beltRold_l_04_a\"),Transform=(Rotation=(X=0,Y=0,Z=0,W=1),Translation=(X=0,Y=0,Z=0),Sc ...(truncated)"
    WeightMinimum=0
    Weight=RigVMDispatchNode(False=1, True=0.6, Condition=RigVMVariableNode() {variable=isPlayingEmote}.Value) {template="DISPATCH_RigVMDispatch_If(in Condition,in True,in False,out Result)"}.Result
    ItemToModify.2.Transform.Rotation=@n7.Result
    ItemToModify.3.Transform.Rotation=@n7.Result
    ItemToModify.4.Transform.Rotation=@n7.Result
    ItemToModify.0.Transform.Rotation=@n8.Result
    ItemToModify.1.Transform.Rotation=@n8.Result
    ExecutePin=@RigVMFunction_Sequence.C
   )
   > &RigUnit_ModifyTransforms(
    ItemToModify=((Item=(Type=Bone,Name="dyn_beltRold_l_06_a"),Transform=(Rotation=(X=0,Y=0,Z=0,W=1),Translation=(X=0,Y=0,Z=0),Scale3D=(X=1,Y=1,Z=1))))
    WeightMinimum=0
    Weight=RigVMDispatchNode(False=0, True=1, Condition=RigVMVariableNode() {variable=isPlayingEmote}.Value) {template="DISPATCH_RigVMDispatch_If(in Condition,in True,in False,out Result)"}.Result
    ItemToModify.0.Transform.Rotation=@n7.Result
   )
   Multiply = &RigVMFunction_MathDoubleMul(
    B=RigVMDispatchNode(BoneName=root, Default=0, Name=deformRig_l_thigh_up_alpha) {template="DISPATCH_RigDispatch_GetAnimAttribute(in Name,in BoneName,in Default,out Value,out Found)"}.Value
    A=RigVMVariableNode() {variable=isMovingCrouching}.Value
   )
   > &RigUnit_ModifyTransforms(
    ItemToModify="((Item=(Type=Bone,Name=\"dyn_beltRold_l_04_a\"),Transform=(Rotation=(X=0,Y=-0.1736,Z=0,W=0.9848),Translation=(X=0,Y=0,Z=0),Scale3D=(X=1,Y=1,Z=1))),(Item=(Type=Bone,Name=\"dyn_beltRold_l_04_b\"),Transform=(Rotation=(X=0,Y=-0.1736,Z=0,W=0.9848),Translation=(X=0, ...(truncated)"
    WeightMinimum=0
    Weight=@Multiply.Result
   )
   n9 = RigVMVariableNode() {variable=Leg_r_Swing_rot}
   n10 = &RigVMFunction_AnimEvalRichCurve(
    Curve="(EditorCurveData=(Keys=((InterpMode=RCIM_Cubic,TangentMode=RCTM_User,TangentWeightMode=RCTWM_WeightedNone,Time=0,Value=0,ArriveTangent=0.2796,ArriveTangentWeight=0,LeaveTangent=0.2796,LeaveTangentWeight=0),(InterpMode=RCIM_Cubic,TangentMode=RCTM_User,Tange ...(truncated)"
    SourceMaximum=-140
    SourceMinimum=0
    TargetMaximum=-44
    TargetMinimum=0
    Value=@n9.Value.Y
   )
   n11 = &RigVMFunction_AnimEvalRichCurve(
    Curve="(EditorCurveData=(Keys=((InterpMode=RCIM_Cubic,TangentMode=RCTM_User,TangentWeightMode=RCTWM_WeightedNone,Time=0,Value=0,ArriveTangent=0.2796,ArriveTangentWeight=0,LeaveTangent=0.2796,LeaveTangentWeight=0),(InterpMode=RCIM_Cubic,TangentMode=RCTM_User,Tange ...(truncated)"
    SourceMaximum=-140
    SourceMinimum=0
    TargetMaximum=-44
    TargetMinimum=0
    Value=@n9.Value.Z
   )
   n12 = &RigVMFunction_AlphaInterp(
    InterpSpeedDecreasing=10
    InterpSpeedIncreasing=10
    Scale=1
    bClampResult=false
    bInterpResult=false
    bMapRange=true
    Value=&RigVMFunction_MathDoubleAdd(A=@n10.Result, B=@n11.Result).Result
   )
   n13 = &RigVMFunction_MathQuaternionFromEuler(Euler=(X=0,Y=0,Z=0), Euler.Y=&RigVMFunction_MathDoubleMul(B=0.7, A=@n12.Result).Result)
   > &RigUnit_ModifyTransforms(
    ItemToModify="((Item=(Type=Bone,Name=\"dyn_beltRold_r_04_a\"),Transform=(Rotation=(X=0,Y=0,Z=0,W=1),Translation=(X=0,Y=0,Z=0),Scale3D=(X=1,Y=1,Z=1))),(Item=(Type=Bone,Name=\"dyn_beltRold_r_03_a\"),Transform=(Rotation=(X=0,Y=0,Z=0,W=1),Translation=(X=0,Y=0,Z=0),Scale3D=(X=1, ...(truncated)"
    WeightMinimum=0
    Weight=RigVMDispatchNode(False=1, True=0.6, Condition=RigVMVariableNode() {variable=isPlayingEmote}.Value) {template="DISPATCH_RigVMDispatch_If(in Condition,in True,in False,out Result)"}.Result
    ItemToModify.1.Transform.Rotation=@n13.Result
    ItemToModify.2.Transform.Rotation=@n13.Result
    ItemToModify.3.Transform.Rotation=@n13.Result
    ItemToModify.0.Transform.Rotation=&RigVMFunction_MathQuaternionFromEuler(Euler=(X=0,Y=0,Z=0), Euler.Y=&RigVMFunction_MathDoubleMul(B=1.5, A=@n12.Result).Result).Result
   )
   > &RigUnit_ModifyTransforms(
    ItemToModify="((Item=(Type=Bone,Name=\"dyn_beltRold_r_04_a\"),Transform=(Rotation=(X=0,Y=-0.0872,Z=0,W=0.9962),Translation=(X=0,Y=0,Z=0),Scale3D=(X=1,Y=1,Z=1))),(Item=(Type=Bone,Name=\"dyn_beltRold_r_04_b\"),Transform=(Rotation=(X=0,Y=-0.1305,Z=0,W=0.9914),Translation=(X=0, ...(truncated)"
    WeightMinimum=0
    Weight=&RigVMFunction_MathDoubleMul(
     B=RigVMDispatchNode(BoneName=root, Default=0, Name=deformRig_r_thigh_up_alpha) {template="DISPATCH_RigDispatch_GetAnimAttribute(in Name,in BoneName,in Default,out Value,out Found)"}.Value
     A=RigVMVariableNode() {variable=isMovingCrouching}.Value
    ).Result
   )
   &RigUnit_ParentConstraint(
    AdvancedSettings=(RotationOrderForFilter=XZY)
    Child=(Type=Bone,Name="dyn_beltRold_r_04_a")
    ExecutePin=()
    Filter=(TranslationFilter=(bX=false,bY=false,bZ=false),RotationFilter=(bX=True,bY=True,bZ=True),ScaleFilter=(bX=false,bY=false,bZ=false))
    Parents=((Item=(Type=Bone,Name="spine_04"),Weight=0.0025),(Item=(Type=Bone,Name="spine_03"),Weight=0.1061),(Item=(Type=Bone,Name="spine_01"),Weight=0.1511),(Item=(Type=Bone,Name="spine_02"),Weight=0.1808),(Item=(Type=Bone,Name="pelvis"),Weight=0.5595))
    Weight=1
   )
   > &RigUnit_ParentConstraint(
    AdvancedSettings=(RotationOrderForFilter=XZY)
    Child=(Type=Bone,Name="dyn_beltRold_r_03_a")
    Filter=(TranslationFilter=(bX=false,bY=false,bZ=false),RotationFilter=(bX=True,bY=True,bZ=True),ScaleFilter=(bX=false,bY=false,bZ=false))
    Parents=((Item=(Type=Bone,Name="spine_03"),Weight=0.0987),(Item=(Type=Bone,Name="spine_01"),Weight=0.1459),(Item=(Type=Bone,Name="spine_02"),Weight=0.1749),(Item=(Type=Bone,Name="pelvis"),Weight=0.5805))
    Weight=1
   )
   > &RigUnit_ParentConstraint(
    AdvancedSettings=(RotationOrderForFilter=XZY)
    Child=(Type=Bone,Name="dyn_beltRold_l_06_a")
    Filter=(TranslationFilter=(bX=false,bY=false,bZ=false),RotationFilter=(bX=True,bY=True,bZ=True),ScaleFilter=(bX=false,bY=false,bZ=false))
    Parents=((Item=(Type=Bone,Name="spine_04"),Weight=0.0127),(Item=(Type=Bone,Name="spine_03"),Weight=0.114),(Item=(Type=Bone,Name="spine_01"),Weight=0.1592),(Item=(Type=Bone,Name="spine_02"),Weight=0.1894),(Item=(Type=Bone,Name="pelvis"),Weight=0.5248))
    Weight=1
   )
   > &RigUnit_ParentConstraint(
    AdvancedSettings=(RotationOrderForFilter=XZY)
    Child=(Type=Bone,Name="dyn_beltRold_l_05_a")
    Filter=(TranslationFilter=(bX=false,bY=false,bZ=false),RotationFilter=(bX=True,bY=True,bZ=True),ScaleFilter=(bX=false,bY=false,bZ=false))
    Parents="((Item=(Type=Bone,Name=\"thigh_twist_01_l\"),Weight=0.0005),(Item=(Type=Bone,Name=\"deform_groin_l\"),Weight=0.0007),(Item=(Type=Bone,Name=\"spine_04\"),Weight=0.0082),(Item=(Type=Bone,Name=\"spine_03\"),Weight=0.1053),(Item=(Type=Bone,Name=\"spine_01\"),Weight=0.14 ...(truncated)"
    Weight=1
   )
   > &RigUnit_ParentConstraint(
    AdvancedSettings=(RotationOrderForFilter=XZY)
    Child=(Type=Bone,Name="dyn_beltRold_l_04_a")
    Filter=(TranslationFilter=(bX=false,bY=false,bZ=false),RotationFilter=(bX=True,bY=True,bZ=True),ScaleFilter=(bX=false,bY=false,bZ=false))
    Parents=((Item=(Type=Bone,Name="spine_04"),Weight=0.0025),(Item=(Type=Bone,Name="spine_03"),Weight=0.1061),(Item=(Type=Bone,Name="spine_01"),Weight=0.1511),(Item=(Type=Bone,Name="spine_02"),Weight=0.1808),(Item=(Type=Bone,Name="pelvis"),Weight=0.5595))
    Weight=1
   )
   > &RigUnit_ParentConstraint(
    AdvancedSettings=(RotationOrderForFilter=XZY)
    Child=(Type=Bone,Name="dyn_beltRold_l_03_a")
    Filter=(TranslationFilter=(bX=false,bY=false,bZ=false),RotationFilter=(bX=True,bY=True,bZ=True),ScaleFilter=(bX=false,bY=false,bZ=false))
    Parents=((Item=(Type=Bone,Name="spine_03"),Weight=0.0987),(Item=(Type=Bone,Name="spine_01"),Weight=0.1459),(Item=(Type=Bone,Name="spine_02"),Weight=0.1749),(Item=(Type=Bone,Name="pelvis"),Weight=0.5805))
    Weight=1
   )
  ]}
 RigVMFunctionLibrary {nodes [
   Belts_Setup = RigVMCollapseNode(ExecuteContext=())
   Shoulders_Setup = RigVMCollapseNode(ExecuteContext=())
  ]}
 RigVMFunction_Sequence_1_ContainedGraph {nodes [
   n1 = RigVMFunctionEntryNode()
   RigVMFunction_Sequence = &RigVMFunction_Sequence(ExecuteContext=@n1.ExecuteContext)
   n2 = &RigVMFunction_Sequence(ExecuteContext=@RigVMFunction_Sequence.B)
   RigVMFunctionReturnNode(A=@RigVMFunction_Sequence.A, B=@n2.A, C=@n2.B)
  ]}
 RigVMFunction_Sequence_1_ContainedGraph {nodes [
   n1 = RigVMFunctionEntryNode()
   RigVMFunction_Sequence = &RigVMFunction_Sequence(ExecuteContext=@n1.ExecuteContext)
   n2 = &RigVMFunction_Sequence(ExecuteContext=@RigVMFunction_Sequence.B)
   n3 = &RigVMFunction_Sequence(ExecuteContext=@n2.B)
   RigVMFunctionReturnNode(A=@RigVMFunction_Sequence.A, B=@n2.A, C=@n3.A, D=@n3.B)
  ]}
 RigVMModel {nodes [
   n1 = &RigUnit_PrepareForExecution()
   &RigUnit_HierarchyImportFromSkeleton(NameSpace=None, bIncludeCurves=False, bIncludeMeshSockets=False, bIncludeVirtualBones=True, ExecutePin=@n1.ExecutePin)
   RigUnit_BeginExecution = &RigUnit_BeginExecution()
   FN_DeformRig_FNC = RigVMFunctionReferenceNode(
    DeformRig_ Curve="(deformRig_l_upperarm_up_alpha_72_#1=true,deformRig_r_upperarm_up_alpha_73_#2=true,deformRig_l_upperarm_fr_alpha_74_#3=true,deformRig_r_upperarm_fr_alpha_75_CB51B80C4 ...(truncated)"
    DeformRig_Bool="(upperarm_1_#4=True,elbow_3_#5=True,wrist_7_#6=True,groin_8_#7=True,glute_17_#8=True,knee_9_40C408324CA261 ...(truncated)"
    DeformRig_Debug="(upperarm_debug=False,wrist_debug=False,lat_debug=False,groinGlute_debug=False,upperarm_debug_18_#4=False,wrist_debug_20_#6=False,groinGlute_debug_27_#7=False,lat_deb ...(truncated)"
    DeformRig_Value="(upperarm_up_18_#9=(Translation=(X=0,Y=0,Z=8),Scale3D=(X=0.7,Y=1,Z=1)),upperarm_fwd_fr_19_#10=(Translation=(X=0,Y=10,Z=0),Scale3D=(X=0.5,Y=1,Z=1)),upperarm_fwd_bk_20_9973AAD84EE0A1E349BC9E85067D4A3 ...(truncated)"
    ExecuteContext=@RigUnit_BeginExecution.ExecutePin
   )
   > Shoulders_Setup = RigVMFunctionReferenceNode()
   > Belts_Setup = RigVMFunctionReferenceNode()
  ]}
]
hierarchy [
 thigh_twist_01_l_ctrl {controlType=EULER_TRANSFORM, type=Control, loc=<0 0 0>, rot=<0 0 0 1>}
 thigh_twist_01_r_ctrl {controlType=EULER_TRANSFORM, type=Control, loc=<0 0 0>, rot=<0 0 0 1>}
]
variables [
 Leg_l_Swing_rot {type=FVector, default=(X=0,Y=0,Z=0)}
 Leg_r_Swing_rot {type=FVector, default=(X=0,Y=0,Z=0)}
 isMovingCrouching {type=double, default=0}
 isPlayingEmote {type=bool, default=False}
]
