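# Animation Blueprint for the SquidGlisten LiftDairy medium male player body cosmetic mesh.
# It copies the pose from the attached parent mesh, then runs the body's own Control Rig on it.
# The rig receives an emote flag and a crouch-walk weight.
# Each update eases that weight toward 1 while crouching and moving in 2D, and back toward 0 otherwise, using FInterpTo.
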AnimBlueprint /BRCosmetics/Characters/Player/Male/Medium/Bodies/M_MED_SquidGlisten_LiftDairy/Meshes/M_MED_SquidGlisten_LiftDairy_AnimBP
target_skeleton=&/M_MED_SquidGlisten_LiftDairy_Skeleton
graphs [
 AnimGraph {nodes [
   n1 = Control Rig!(
    Source=Copy Pose From Mesh() {bUseAttachedParent=true}
    Alpha
    isMovingCrouching:double=Get Is Crouch Walking() -> (Is Crouch Walking:double)
    isPlayingEmote:bool=Get bIsPlayingEmote()
   ) -> (Pose) {
    alphaBoolBlend {blendOption=Linear}
    controlRigAssetReference {blueprintRigClass=&/M_MED_SquidGlisten_LiftDairy_CtrlRig.M_MED_SquidGlisten_LiftDairy_CtrlRig_C}
   }
   Output Pose(Result=@n1) {layerGroup=None}
  ]}
 EventGraph {nodes [
   n1 = Event BlueprintUpdateAnimation()
   > Set deltaTime(deltaTime:double=@n1.DeltaTimeX) -> (Output_Get:double)
   n2 = AND Boolean(A=Get bIsCrouching(), B=Get bIsMoving2D())
   > Set Is Crouch Walking(Is Crouch Walking:double=FInterpTo(Current=Get Is Crouch Walking(), Target="To Float (Boolean)"(InBool=@n2), DeltaTime=Get deltaTime() -> (deltaTime:double), InterpSpeed=2)) -> (Output_Get:double)
  ]}
]
variables [
 deltaTime {type=number}
 Is Crouch Walking {type=number}
]
